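# Physics asset for the FacetWhisk gadget skeletal mesh, a cosmetic emote prop.
# A single capsule body on the root bone covers the prop, with no per-part bodies and no constraints, so it has no ragdoll articulation.
# The capsule's length is negligible next to its very large radius of roughly 553 units, making it effectively a big sphere.
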PhysicsAsset /CosmeticCompanions/Assets/Props/Emotes/FacetWhisk/Meshes/SKM_EmoteProp_FacetWhisk_Gadget_Physics
bodies {root [{shapeType=Capsule, length=0.5049, radius=552.889, rotation=<rot 0 -90 0>}]}
constraints []
